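# Post-copy animation Blueprint for the female medium TaskForce body cosmetic, adding secondary motion to its gear.
# Copies the attached parent's pose into a Control Rig driven by the parachute-open flag, then blends the left walkie-talkie bone between spine and clavicle through a constraint.
# AnimDynamics swings the right pouch with an angular spring and a thigh-driven planar limit.
# Each update refreshes the parachute flag and a gravity override variable.
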AnimBlueprint /BRCosmetics/Characters/Player/Female/Medium/Bodies/F_MED_TaskForce/Meshes/F_MED_TaskForce_AnimBP
target_skeleton=&/F_MED_TaskForce_Skeleton
graphs [
 AnimGraph {nodes [
   n1 = Control Rig!(Source=Copy Pose From Mesh() {bUseAttachedParent=true}, Alpha, Parachute_open:bool=Get Parachute_open() -> (Parachute_open:bool)) -> (Pose) {
    alphaBoolBlend {blendOption=Linear}
    controlRigAssetReference {blueprintRigClass=&/F_MED_TaskForce_CtrlRig.F_MED_TaskForce_CtrlRig_C}
   }
   n2 = "Constraint - base_walkietalkie_01_l (PP)"!(ComponentPose=Local To Component(LocalPose=@n1), Alpha=1, ConstraintWeights_0=0.4, ConstraintWeights_1:float=0.6) -> (Pose) {
    alphaBoolBlend {blendOption=Linear}
    boneToModify {boneName=base_walkietalkie_01_l}
    constraintSetup [{targetBone {boneName=spine_05}}, {targetBone {boneName=clavicle_l}}]
    constraintWeights [1,1]
   }
   n3 = Anim Dynamics - Bone: dyn_Pouch_01_r!(ComponentPose=@n2, Alpha=1, GravityOverride:Vector="0,0,-980", ExternalForce="0,0,0") -> (Pose) {
    angularDampingOverride=1
    angularSpringConstant=10000
    bAngularSpring=true
    bOverrideAngularDamping=true
    bUseGravityOverride=true
    gravityScale=0.35
    simulationSpace=RootRelative
    alphaBoolBlend {blendOption=Linear}
    boundBone {boneName=dyn_Pouch_01_r}
    componentLinearAccScale=<1 1 1>
    componentLinearVelScale=<1 1 1>
    gravityOverride {z=-980}
    physicsBodyDefinitions [{
      boundBone {boneName=dyn_Pouch_01_r}
      boxExtents=<20 30 30>
      constraintSetup {angularLimitsMax=<0.5 2 0>, angularLimitsMin=<-0.5 -2 0>, angularTarget=<1 0 0>}
      localJointOffset=<2.3749 0 0>
     }]
    planarLimits [{drivingBone {boneName=thigh_r}, planeTransform {location=<0 0 7.0355>, rotation=<rot 0 0 0>}}]
   }
   Output Pose(Result=Component To Local(ComponentPose=@n3)) {layerGroup=None}
  ]}
 EventGraph {nodes [
   Event BlueprintUpdateAnimation()
   > n1 = "Gravity Override State Function (Parameter Asset) (Nativized)"(CharacterPartAnimInstance=Self-Reference(), GameTimeInSeconds=GetGameTimeInSeconds())
   > Set GravityOverride(GravityOverride=@n1)
   > Set Parachute_open(Parachute_open:bool=Get bIsParachuteOpened()) -> (Output_Get:bool)
  ]}
]
variables [
 GravityOverride {type=Vector}
 Parachute_open {type=boolean}
]
