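# Physics asset for the F_MED_TaskForce female player body, covering a dynamic right-side pouch bone and the body parts it hangs from.
# Capsule bodies sit on the pelvis, right thigh and dyn_Pouch_01_r, forming a short chain rather than a full-body ragdoll.
# The pouch is constrained to the thigh with near-zero angular limits, allowing only slight swing.
# The thigh-to-pelvis constraint leaves swing and twist free.
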PhysicsAsset /BRCosmetics/Characters/Player/Female/Medium/Bodies/F_MED_TaskForce/Meshes/F_MED_TaskForce_Physics
bodies {
 dyn_Pouch_01_r [{shapeType=Capsule, length=9.2015, radius=6.5979, shapeName=dyn_Pouch_01_r_capsule, center=<2.1221 0 5.1>, rotation=<rot 0 89.88 -89.4492>}]
 pelvis [{shapeType=Capsule, length=10.2554, radius=7.96, shapeName=pelvis_capsule, center=<-0.1031 -1.76 0.3799>, rotation=<rot 0 170.6453 -174.7432>}]
 thigh_r [{
   shapeType=Capsule
   length=21.9977
   radius=9.3954
   shapeName=thigh_r_capsule
   center=<20.45 -0.7346 -2.733>
   rotation=<rot -0.358 86.7277 -93.9013>
  }]
}
constraints [
 {bone1Name=dyn_Pouch_01_r, bone2Name=thigh_r, swing1LimitDegrees=0, swing2LimitDegrees=2, twistLimitDegrees=0}
 {
  bone1Name=thigh_r
  bone2Name=pelvis
  swing1LimitDegrees=45
  swing1Motion=Free
  swing2LimitDegrees=45
  swing2Motion=Free
  twistLimitDegrees=45
  twistMotion=Free
 }
]
