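# Physics asset for the face accessory mesh of the female medium TaskForce body, covering a right-side communicator and a hair bun.
# Single capsules stand in for the head, the communicator bone and the hair bun bone.
# Both dyn_ bones are constrained directly to the head: the communicator has very tight swing limits with twist locked, while the bun allows wider swing with a small twist range.
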PhysicsAsset /BRCosmetics/Characters/Player/Female/Medium/Bodies/F_MED_TaskForce/Meshes/Parts/F_MED_TaskForce_FaceAcc_Physics
bodies {
 dyn_communicator_02_r [{
   shapeType=Capsule
   length=6.8065
   radius=1.0179
   shapeName=dyn_communicator_02_r_capsule
   center=<3.4109 -1.586 -0.3335>
   rotation=<rot -0.0017 -131.8526 -81.9243>
  }]
 dyn_hair_bun_01 [{
   shapeType=Capsule
   length=0.375
   radius=4.38
   shapeName=dyn_hair_bun_01_capsule
   center=<5.802 -0.0253 -0.7803>
   rotation=<rot 0 -85 90.0003>
  }]
 head [{
   shapeType=Capsule
   length=8.9558
   radius=5.9259
   shapeName=head_offset_capsule
   center=<0.5659 2.1774 -0.7524>
   rotation=<rot 0 -132.099 -106.1971>
  }]
}
constraints [
 {bone1Name=dyn_communicator_02_r, bone2Name=head, swing1LimitDegrees=2, swing2LimitDegrees=1, twistLimitDegrees=45, twistMotion=Locked}
 {bone1Name=dyn_hair_bun_01, bone2Name=head, swing1LimitDegrees=15, swing2LimitDegrees=20, twistLimitDegrees=2}
]
