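# Post-copy animation blueprint for the female TaskForce face accessory, adding secondary physics motion to its hair.
# It copies the attached parent's pose, layers the Body part's pose from clavicles and upper arms, then caches it.
# A head-space RigidBody simulation feeds chained, spring-driven AnimDynamics nodes for bun and front strands, some with head-sphere limits.
# The event graph captures Body on customization load and updates a GravityOverride variable each frame, with no visible AnimGraph connection.
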AnimBlueprint /BRCosmetics/Characters/Player/Female/Medium/Bodies/F_MED_TaskForce/Meshes/Parts/F_MED_TaskForce_FaceAcc_AnimBP
target_skeleton=&/F_MED_TaskForce_FaceAcc_Skeleton
graphs [
 AnimGraph {nodes [
   n1 = RigidBody!(ComponentPose=Local To Component(LocalPose=Use cached pose 'InitialPose'()), Alpha, OverrideWorldGravity:Vector="0,0,-980", ExternalForce) -> (Pose) {
    bOverrideWorldGravity=true
    overridePhysicsAsset=&/F_MED_TaskForce_FaceAcc_Physics
    simulationSpace=BaseBoneSpace
    alphaBoolBlend {blendOption=Linear}
    baseBoneRef {boneName=head}
    simSpaceSettings {maxAngularAcceleration=40, maxAngularVelocity=8, maxLinearAcceleration=3500, maxLinearVelocity=700, worldAlpha=0.5}
   }
   n2 = Anim Dynamics - Bone: dyn_hair_bun_01!(ComponentPose=@n1, Alpha=1, GravityOverride:Vector="0,0,-980", ExternalForce="0,0,0") -> (Pose) {
    angularDampingOverride=1
    angularSpringConstant=15000
    bAngularSpring=true
    bOverrideAngularDamping=true
    bUseGravityOverride=true
    gravityScale=0.35
    simulationSpace=RootRelative
    alphaBoolBlend {blendOption=Linear}
    boundBone {boneName=dyn_hair_bun_01}
    componentLinearAccScale=<1 1 1>
    componentLinearVelScale=<1 1 1>
    gravityOverride {z=-980}
    physicsBodyDefinitions [{
      boundBone {boneName=dyn_hair_bun_01}
      boxExtents=<20 30 30>
      constraintSetup {angularLimitsMax=<15 15 15>, angularLimitsMin=<-15 -15 -15>, angularTarget=<1 0 0>}
      localJointOffset=<4 0 0>
     }]
   }
   n3 = Anim Dynamics - Bone: dyn_front_hair_01_l!(ComponentPose=@n2, Alpha=1, GravityOverride:Vector="0,0,-980", ExternalForce="0,0,0") -> (Pose) {
    angularDampingOverride=1
    angularSpringConstant=10000
    bAngularSpring=true
    bOverrideAngularDamping=true
    bUseGravityOverride=true
    bUseSphericalLimits=true
    gravityScale=0.35
    simulationSpace=RootRelative
    alphaBoolBlend {blendOption=Linear}
    boundBone {boneName=dyn_front_hair_01_l}
    componentLinearAccScale=<1 1 1>
    componentLinearVelScale=<1 1 1>
    gravityOverride {z=-980}
    physicsBodyDefinitions [{
      boundBone {boneName=dyn_front_hair_01_l}
      boxExtents=<20 30 30>
      constraintSetup {angularLimitsMax=<0 0 2>, angularLimitsMin=<0 0 0>, angularTarget=<1 0 0>}
      localJointOffset=<4 0 0>
     }]
    sphericalLimits [{limitRadius=7, drivingBone {boneName=head}, sphereLocalOffset=<5.0203 4.6811 0.3941>}]
   }
   n4 = Anim Dynamics - Bone: dyn_front_hair_02_l!(ComponentPose=@n3, Alpha=1, GravityOverride:Vector="0,0,-980", ExternalForce="0,0,0") -> (Pose) {
    angularDampingOverride=1
    angularSpringConstant=10000
    bAngularSpring=true
    bOverrideAngularDamping=true
    bUseGravityOverride=true
    bUseSphericalLimits=true
    gravityScale=0.35
    simulationSpace=RootRelative
    alphaBoolBlend {blendOption=Linear}
    boundBone {boneName=dyn_front_hair_02_l}
    componentAppliedLinearAccClamp=<5000 5000 5000>
    componentLinearAccScale=<0.2 0.2 0.2>
    componentLinearVelScale=<0.2 0.2 0.2>
    gravityOverride {z=-980}
    physicsBodyDefinitions [{
      boundBone {boneName=dyn_front_hair_02_l}
      boxExtents=<20 30 30>
      constraintSetup {angularLimitsMax=<0 2 0>, angularLimitsMin=<0 -2 0>, angularTarget=<1 0 0>}
      localJointOffset=<4 0 0>
     }]
    sphericalLimits [{limitRadius=7, drivingBone {boneName=head}, sphereLocalOffset=<1.9176 4.4983 0.2023>}]
   }
   n5 = Anim Dynamics - Bone: dyn_front_hair_03_l!(ComponentPose=@n4, Alpha=1, GravityOverride:Vector="0,0,-980", ExternalForce="0,0,0") -> (Pose) {
    angularDampingOverride=1
    angularSpringConstant=10000
    bAngularSpring=true
    bOverrideAngularDamping=true
    bUseGravityOverride=true
    bUseSphericalLimits=true
    gravityScale=0.35
    simulationSpace=RootRelative
    alphaBoolBlend {blendOption=Linear}
    boundBone {boneName=dyn_front_hair_03_l}
    componentAppliedLinearAccClamp=<5000 5000 5000>
    componentLinearAccScale=<0.2 0.2 0.2>
    componentLinearVelScale=<0.2 0.2 0.2>
    gravityOverride {z=-980}
    physicsBodyDefinitions [{
      boundBone {boneName=dyn_front_hair_03_l}
      boxExtents=<20 30 30>
      constraintSetup {angularLimitsMax=<0 4 0>, angularLimitsMin=<0 -4 0>, angularTarget=<1 0 0>}
      localJointOffset=<4 0 0>
     }]
    sphericalLimits [{limitRadius=9, drivingBone {boneName=head}, sphereLocalOffset=<0 0 0>}]
   }
   n6 = Anim Dynamics - Bone: dyn_front_hair_04_l!(ComponentPose=@n5, Alpha=1, GravityOverride:Vector="0,0,-980", ExternalForce="0,0,0") -> (Pose) {
    angularDampingOverride=1
    angularSpringConstant=10000
    bAngularSpring=true
    bOverrideAngularDamping=true
    bUseGravityOverride=true
    bUseSphericalLimits=true
    gravityScale=0.35
    simulationSpace=RootRelative
    alphaBoolBlend {blendOption=Linear}
    boundBone {boneName=dyn_front_hair_04_l}
    componentAppliedLinearAccClamp=<5000 5000 5000>
    componentLinearAccScale=<0.2 0.2 0.2>
    componentLinearVelScale=<0.2 0.2 0.2>
    gravityOverride {z=-980}
    physicsBodyDefinitions [{
      boundBone {boneName=dyn_front_hair_04_l}
      boxExtents=<20 30 30>
      constraintSetup {angularLimitsMax=<0 6 0>, angularLimitsMin=<0 -6 0>, angularTarget=<1 0 0>}
      localJointOffset=<4 0 0>
     }]
    sphericalLimits [{limitRadius=10, drivingBone {boneName=head}, sphereLocalOffset=<0 0 0>}]
   }
   n7 = Anim Dynamics - Bone: dyn_front_hair_c_r!(ComponentPose=@n6, Alpha=1, GravityOverride:Vector="0,0,0", ExternalForce="0,0,0") -> (Pose) {
    angularSpringConstant=5000
    bAngularSpring=true
    bUseGravityOverride=true
    simulationSpace=RootRelative
    alphaBoolBlend {blendOption=Linear}
    boundBone {boneName=dyn_front_hair_c_r}
    physicsBodyDefinitions [{
      boundBone {boneName=dyn_front_hair_c_r}
      boxExtents=<10 10 10>
      constraintSetup {angularLimitsMax=<0 3 0>, angularLimitsMin=<0 -3 0>, angularTarget=<0 0 0>}
      localJointOffset=<0 0 0>
     }]
   }
   n8 = Anim Dynamics - Bone: dyn_front_hair_b_01_r!(ComponentPose=@n7, Alpha=1, GravityOverride:Vector="0,0,0", ExternalForce="0,0,0") -> (Pose) {
    angularSpringConstant=5000
    bAngularSpring=true
    bUseGravityOverride=true
    simulationSpace=RootRelative
    alphaBoolBlend {blendOption=Linear}
    boundBone {boneName=dyn_front_hair_b_01_r}
    physicsBodyDefinitions [{
      boundBone {boneName=dyn_front_hair_b_01_r}
      boxExtents=<10 10 10>
      constraintSetup {angularLimitsMax=<0 3 0>, angularLimitsMin=<0 -3 0>, angularTarget=<0 0 0>}
      localJointOffset=<0 0 0>
     }]
   }
   n9 = Anim Dynamics - Bone: dyn_front_hair_inner_b_r!(ComponentPose=@n8, Alpha=1, GravityOverride:Vector="0,0,0", ExternalForce="0,0,0") -> (Pose) {
    angularSpringConstant=5000
    bAngularSpring=true
    bUseGravityOverride=true
    bUseSphericalLimits=true
    simulationSpace=RootRelative
    alphaBoolBlend {blendOption=Linear}
    boundBone {boneName=dyn_front_hair_inner_b_r}
    physicsBodyDefinitions [{
      boundBone {boneName=dyn_front_hair_inner_b_r}
      boxExtents=<10 10 10>
      constraintSetup {angularLimitsMax=<0 0 0>, angularLimitsMin=<0 -2 0>, angularTarget=<0 0 0>}
      localJointOffset=<0 0 0>
     }]
    sphericalLimits [{limitRadius=8.5, drivingBone {boneName=head}, sphereLocalOffset=<3.2309 1.3812 -0.6216>}]
   }
   n10 = Anim Dynamics - Bone: base_front_hair_d_01_r!(ComponentPose=@n9, Alpha=1, GravityOverride:Vector="0,0,0", ExternalForce="0,0,0") -> (Pose) {
    angularSpringConstant=5000
    bAngularSpring=true
    bUseGravityOverride=true
    bUseSphericalLimits=true
    simulationSpace=RootRelative
    alphaBoolBlend {blendOption=Linear}
    boundBone {boneName=base_front_hair_d_01_r}
    physicsBodyDefinitions [{
      boundBone {boneName=base_front_hair_d_01_r}
      boxExtents=<10 10 10>
      constraintSetup {angularLimitsMax=<0 0 2>, angularLimitsMin=<0 -1 -1>, angularTarget=<0 0 0>}
      localJointOffset=<0 0 0>
     }]
    sphericalLimits [{limitRadius=8.5, drivingBone {boneName=head}, sphereLocalOffset=<5.7505 0.6565 -1.0353>}]
   }
   Output Pose(Result=Component To Local(ComponentPose=@n10)) {layerGroup=None}
   n11 = Layered blend per bone(
    BasePose=Copy Pose From Mesh() {bUseAttachedParent=true}
    BlendPoses_0=Copy Pose From Mesh(SourceMeshComponent=Get Body() -> (Body:Skeletal Mesh Component Budgeted Object Reference)) {bUseAttachedParent=true}
   ) {layerSetup [{branchFilters [
       {blendDepth=2, boneName=clavicle_r}
       {blendDepth=2, boneName=clavicle_l}
       {blendDepth=2, boneName=upperarm_l}
       {blendDepth=2, boneName=upperarm_r}
      ]}]}
   Save cached pose 'InitialPose'(Pose=@n11) {cacheName=InitialPose}
  ]}
 EventGraph {nodes [
   Event BlueprintUpdateAnimation()
   > n1 = "Gravity Override State Function (Parameter Asset) (Nativized)"(CharacterPartAnimInstance=Self-Reference(), GameTimeInSeconds=GetGameTimeInSeconds())
   > Set GravityOverride(GravityOverride=@n1)
   n2 = Event OnCharacterCustomizationLoaded()
   > Set Body(Body:Skeletal Mesh Component Budgeted Object Reference=Get Array Item with EFort Custom Part Type(Array=@n2.CustomMeshParts)) -> (Output_Get:Skeletal Mesh Component Budgeted Object Reference)
  ]}
]
variables [
 Body {type=SkeletalMeshComponentBudgeted}
 GravityOverride {type=Vector}
]
